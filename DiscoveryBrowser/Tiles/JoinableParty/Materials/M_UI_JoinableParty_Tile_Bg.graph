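# Translucent UI material for the Discovery browser's joinable party tile background, with a static switch between thumbnail artwork and an alternate fill.
# The thumbnail repeats twice horizontally with an X offset and scale-to-fill UVs, with opacity and radial-mask controls.
# A two-color gradient, patterned backing and rounded-corner mask define further layers.
# A delayed, sine-eased screen-space sweep drives the flare mask that tints the outline and feeds the flare highlight.
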Material /DiscoveryBrowser/Tiles/JoinableParty/Materials/M_UI_JoinableParty_Tile_Bg
graph {
 properties {
  bUsedWithStaticMesh=false
  blendMode=BLEND_Translucent
  materialDomain=MD_UI
  pixelDepthOffsetMode=PDOM_Legacy
  usedShadingModels=Default Lit
 }
 nodes [
  n1 = NamedRerouteUsage()
  n2 = NamedRerouteUsage()
  n3 = NamedRerouteUsage()
  LinearInterpolate(A=@n1, B=@n2, Alpha=Round(None=ComponentMask(None=@n3) {r=true}))
  n4 = NamedRerouteUsage()
  n5 = Multiply(A=NamedRerouteUsage(), B=ScalarParameter() {defaultValue=1, group=2 - Flare, parameterName=Flare Opacity, sliderMax=1, sortPriority=0})
  Multiply(A=NamedRerouteUsage(), B=LinearInterpolate(A=@n5, Alpha=OneMinus(None=@n4)))
  n6 = MaterialFunctionCall() {materialFunction=&GetUserInterfaceUV}
  NamedRerouteDeclaration(None=@n6.Normalized UV) {name=UV, nodeColor=<rgba 0 0 0 1>}
  n7 = TextureObjectParameter() {group=0 - Thumbnail, parameterName=Thumbnail, sortPriority=0}
  NamedRerouteDeclaration(None=TextureProperty(Tex=@n7)) {name=Texture Size, nodeColor=<rgba 0 0 0 1>}
  n8 = MaterialFunctionCall("Value (0-1)"=MaterialFunctionCall!(Animation Duration=Constant() {r=1}, Delay=Constant() {r=1}, Reset to Default) {materialFunction=/FortUILibrary/MaterialFunctions/Time/MF_UI_AnimationDelay}.Result) {materialFunction=&MF_UI_EaseCurves}
  n9 = Subtract(A=ScreenPosition().ViewportUV, B=AppendVector(A=LinearInterpolate(Alpha=@n8.EaseOutSine) {constA=-1}, B=Constant()))
  n10 = MaterialFunctionCall!(
   UVs=@n9
   Alpha=ComponentMask(None=MaterialFunctionCall!(Value=@n9, Scale, Height) {materialFunction=/FortUILibrary/MaterialFunctions/Basics/MF_UI_PingPong}.Result) {r=true}
  ) {materialFunction=/FortUILibrary/MaterialFunctions/Basics/MF_UI_ClipEdgePixel}
  NamedRerouteDeclaration(None=@n10.Result) {name=Flare Mask, nodeColor=<rgba 0 0 0 1>}
  NamedRerouteDeclaration(None=MaterialFunctionCall(Corner Radius=Constant() {r=20}, Radius in Pixels=Constant() {r=1}) {materialFunction=&MF_UI_CoreUI_RoundedMask}.Shape Mask) {name=Rounded Corner Mask, nodeColor=<rgba 0 0 0 1>}
  n11 = NamedRerouteUsage()
  n12 = Constant() {r=2}
  n13 = AppendVector(A=@n12, B=Constant() {r=1})
  n14 = ScalarParameter() {parameterName=X Offset, sliderMax=1}
  n15 = Constant()
  n16 = Frac(None=Add(A=Multiply(A=@n11, B=@n13), B=AppendVector(A=@n14, B=@n15)))
  n17 = NamedRerouteUsage()
  n18 = AppendVector(A=Multiply(A=ComponentMask(None=@n17) {r=true}, B=@n12) {constB=2}, B=ComponentMask(None=@n17) {g=true})
  NamedRerouteDeclaration(None=TextureSample(
    UVs=MaterialFunctionCall(UV=@n16, TextureSize=@n18) {materialFunction=&MF_UI_UVScaling}.Scaled To Fill
    Tex=@n7
   ) {samplerSource=SSM_Clamp_WorldGroupSettings}.RGB) {name=Thumbnail Pass, nodeColor=<rgba 0 0 0 1>}
  NamedRerouteDeclaration(None=MaterialFunctionCall!(
    UV=MaterialFunctionCall!(ScreenUV, TextureSize, ScalePivotPoint) {materialFunction=/FortUILibrary/MaterialFunctions/Basics/MF_UI_ScreenSpaceUVScaling}.Scaled to Fit
    Color A
    Color B
   ) {materialFunction=/FortniteCoreUI/UI/DiscoveryShared/Art/MaterialFunctions/MF_UI_Discovery_TileBacking}.Masked Pattern) {name=Backing Visual, nodeColor=<rgba 0 0 0 1>}
  n19 = MaterialFunctionCall(UVs=NamedRerouteUsage(), Texture Scale=Constant() {r=1.5}) {materialFunction=&ScaleUVsByCenter}
  n20 = ScalarParameter() {defaultValue=0.5, parameterName=Mask Direction, sliderMax=0.5, sliderMin=-0.5}
  n21 = Constant()
  n22 = MaterialFunctionCall(UV=Subtract(
    A=MaterialFunctionCall(UV=@n19.UVs) {materialFunction=&MF_UI_UVScaling}.Scaled To Fill
    B=AppendVector(A=@n20, B=@n21)
   )) {materialFunction=&MF_UI_RadialGradient}
  NamedRerouteDeclaration(None=LinearInterpolate(A=Constant() {r=0.9}, B=OneMinus(None=@n22.Smooth), Alpha=ScalarParameter() {defaultValue=1, parameterName=Radial Mask Opacity})) {name=Radial Mask, nodeColor=<rgba 0 0 0 1>}
  n23 = VectorParameter() {group=1 - Background, parameterName=Background Color A, sortPriority=0, defaultValue=<rgba 0.0024 0.0513 0.3864 1>}
  n24 = VectorParameter() {group=1 - Background, parameterName=Background Color B, sortPriority=1, defaultValue=<rgba 0 0.0908 0.9911 1>}
  n25 = ComponentMask(None=NamedRerouteUsage()) {g=true}
  NamedRerouteDeclaration(None=LinearInterpolate(A=@n23.RGB, B=@n24.RGB, Alpha=OneMinus(None=@n25))) {name=Tile Background, nodeColor=<rgba 0 0 0 1>}
  NamedRerouteDeclaration(None=@n6.9-Slice UV) {name=9SliceUV, nodeColor=<rgba 0 0 0 1>}
  n26 = LinearInterpolate(
   A=Multiply(A=NamedRerouteUsage(), B=@n1)
   B=@n2
   Alpha=ScalarParameter() {defaultValue=1, group=0 - Thumbnail, parameterName=Thumbnail Opacity, sliderMax=1}
  )
  n27 = Constant()
  n28 = NamedRerouteUsage()
  n29 = NamedRerouteDeclaration(None=StaticSwitchParameter(True=LinearInterpolate(A=@n26, B=@n27, Alpha=@n28), False=NamedRerouteUsage()) {defaultValue=true, parameterName=Display Thumbnail}) {name=Tile Color, nodeColor=<rgba 0 0 0 1>}
  n30 = Multiply(A=@n29, B=@n5)
  NamedRerouteDeclaration(None=Saturate(None=Add(A=@n29, B=@n30))) {name=Final Color, nodeColor=<rgba 0 0 0 1>}
  NamedRerouteDeclaration(None=LinearInterpolate(A=NamedRerouteUsage(), B=Constant3Vector() {desc=Flare Color, constant=<rgba 0.6038 1 1 1>}, Alpha=@n10.Result)) {name=Outline Visual, nodeColor=<rgba 0 0 0 1>}
  NamedRerouteUsage()
  NamedRerouteUsage()
 ]
}
