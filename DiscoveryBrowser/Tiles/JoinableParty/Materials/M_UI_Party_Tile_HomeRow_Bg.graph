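# UI material for the background of a joinable party tile on the Discovery Browser home row, a translucent rounded panel with a tinted glow.
# The glow mask blends a vertical UV gradient with two offset, stretched radial gradients.
# That mask, weighted by Glow Opacity, lerps Background Color toward Glow Color and feeds a translucency term clipped by a rounded-corner mask.
# An Anim_Invite parameter is declared but connects to nothing shown.
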Material /DiscoveryBrowser/Tiles/JoinableParty/Materials/M_UI_Party_Tile_HomeRow_Bg
graph {
 properties {
  bUsedWithStaticMesh=false
  blendMode=BLEND_Translucent
  materialDomain=MD_UI
  pixelDepthOffsetMode=PDOM_Legacy
  usedShadingModels=Default Lit
 }
 nodes [
  n1 = Constant() {r=0.4}
  n2 = NamedRerouteUsage()
  n3 = VectorParameter() {parameterName=Corner Radius, defaultValue=<rgba 20 20 20 20>}
  LinearInterpolate(
   B=Saturate(None=Add(A=@n1, B=@n2))
   Alpha=MaterialFunctionCall(Corner Radius=AppendVector(A=@n3.RGB, B=@n3.RGB), Radius in Pixels=Constant() {r=1}) {materialFunction=&MF_UI_CoreUI_RoundedMask}.Shape Mask
  ) {constB=0.4}
  LinearInterpolate(
   A=VectorParameter() {parameterName=Background Color, defaultValue=<rgba 0.0545 0.0666 0.1046 1>}.RGB
   B=VectorParameter() {parameterName=Glow Color, defaultValue=<rgba 0.0369 0.9823 0.227 1>}.RGB
   Alpha=Multiply(A=NamedRerouteUsage(), B=ScalarParameter() {defaultValue=1, parameterName=Glow Opacity, sliderMax=1})
  )
  n4 = MaterialFunctionCall() {materialFunction=&GetUserInterfaceUV}
  NamedRerouteDeclaration(None=@n4.Normalized UV) {name=UV, nodeColor=<rgba 0 0 0 1>}
  n5 = Multiply(A=ComponentMask(None=NamedRerouteUsage()) {g=true}) {constB=0.5}
  n6 = Subtract(A=NamedRerouteUsage(), B=Constant2Vector() {g=0.5})
  n7 = MaterialFunctionCall(UV=MaterialFunctionCall(UVs=@n6, Texture Scale=Constant2Vector() {g=1.5, r=2}) {materialFunction=&ScaleUVsByCenter}.UVs) {materialFunction=&MF_UI_RadialGradient}
  n8 = Constant2Vector() {g=1.75, r=0.5}
  n9 = LinearInterpolate(
   A=Multiply(A=LinearInterpolate(A=@n5, Alpha=@n7.Smooth)) {constB=0.4}
   Alpha=MaterialFunctionCall(UV=MaterialFunctionCall(UVs=@n6, Texture Scale=@n8) {materialFunction=&ScaleUVsByCenter}.UVs) {materialFunction=&MF_UI_RadialGradient}.Smooth
  ) {constB=0.6}
  n10 = ScalarParameter() {defaultValue=1, parameterName=Glow Opacity, sliderMax=1}
  NamedRerouteDeclaration(None=Multiply(A=NamedRerouteDeclaration(None=@n9) {name=Glow Mask, nodeColor=<rgba 0 0 0 1>}, B=LinearInterpolate(Alpha=@n10) {constB=0.1}) {constB=0.2}) {name=Glow Opacity, nodeColor=<rgba 0 0 0 1>}
  NamedRerouteDeclaration(None=@n4.9-Slice UV) {name=9SliceUV, nodeColor=<rgba 0 0 0 1>}
  ScalarParameter() {parameterName=Anim_Invite, sliderMax=1}
 ]
}
