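# WBP_JoinableParty_State_Private is a Discovery browser UserWidget, named as the private state of a joinable party tile.
# An Overlay root named InviteGrid holds a single UIKit status indicator set to the Utility type.
# The widget is layout only, with an empty event graph and no Blueprint variables beyond the exposed indicator.
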WidgetBlueprint /DiscoveryBrowser/Tiles/JoinableParty/WBP_JoinableParty_State_Private
graphs [EventGraph {links [], nodes []}]
variables []
widgets {
 info {inheritedWidgetCount=0, parentClass=&UserWidget, rootWidgetClass=&Overlay, widgetCount=2}
 widgets [
  InviteGrid {widgetClassPath=&Overlay, widget=InviteGrid}
  WBP_UIKit_StatusIndicator {
   widgetClassPath=/FortUILibrary/UIKit/StatusIndicator/WBP_UIKit_StatusIndicator
   bIsVariable=true
   parent=InviteGrid
   widget=WBP_UIKit_StatusIndicator
   properties {type=Utility}
   slot {class=OverlaySlot}
  }
 ]
}
