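# Discovery browser tile for a friend's joinable party, showing avatars, names, presence, invite and privacy states, and a rating tag.
# Parent-class events drive text and visibility, while a dynamic background material handles thumbnail fade, invite flare and Arabic mask direction.
# Selection routes through a nested button group view model when present; hover animation swaps presence for a CTA button.
# Mobile touch resizes the tile and hides the CTA; a context menu opens the sidebar.
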WidgetBlueprint /DiscoveryBrowser/Tiles/JoinableParty/WBP_ActivityBrowserTile_JoinableParty_RoundedCorners
animations [
 Anim_OnHoverSelect {
  binding_count=3
  length_seconds=0.25
  object=&self:Anim_OnHoverSelect
  sequence {bindings [
    WBP_CTA_ButtonPrimary {
     id=#1
     possessed_class=WBP_UIKit_Button_Quiet_C
     tracks [
      {
       type=MovieScene2DTransformTrack
       sections [{
         type=MovieScene2DTransformSection
         end_frame=4
         start_frame=0
         channels [
          Translation.X {type=MovieSceneScriptingFloatChannel, const [0,2]}
          Translation.Y {type=MovieSceneScriptingFloatChannel, keys [[0,40],[4,-8]]}
          Angle {type=MovieSceneScriptingFloatChannel}
          Scale.X {type=MovieSceneScriptingFloatChannel}
          Scale.Y {type=MovieSceneScriptingFloatChannel}
          Shear.X {type=MovieSceneScriptingFloatChannel}
          Shear.Y {type=MovieSceneScriptingFloatChannel}
         ]
        }]
      }
      {
       type=MovieSceneFloatTrack
       display_name=Render Opacity
       sections [{type=MovieSceneFloatSection, end_frame=4, start_frame=0, channels [None {type=MovieSceneScriptingFloatChannel, keys [[0,0],[4,1]]}]}]
      }
     ]
    }
    HB_PresenceTexts {
     id=#2
     possessed_class=HorizontalBox
     tracks [{
       type=MovieSceneFloatTrack
       display_name=Render Opacity
       sections [{type=MovieSceneFloatSection, end_frame=4, start_frame=0, channels [None {type=MovieSceneScriptingFloatChannel, keys [[0,1],[4,0]]}]}]
      }]
    }
    Bg {
     id=#3
     possessed_class=Image
     tracks [{
       type=MovieSceneWidgetMaterialTrack
       sections [{
         type=MovieSceneParameterSection
         channels [
          Background Color BR {type=MovieSceneScriptingFloatChannel, keys [[0,0.1022],[5,0.0212]]}
          Background Color BG {type=MovieSceneScriptingFloatChannel, keys [[0,0.0232],[5,0]]}
          Background Color BB {type=MovieSceneScriptingFloatChannel, keys [[0,0.3916],[5,0.1119]]}
          Background Color BA {type=MovieSceneScriptingFloatChannel, const [1,2]}
         ]
        }]
      }]
    }
   ]}
 }
 Anim_OnImageLoaded {
  binding_count=1
  length_seconds=0.2
  object=&self:Anim_OnImageLoaded
  sequence {
   bindings [Bg {
     id=#4
     possessed_class=Image
     tracks [{
       type=MovieSceneWidgetMaterialTrack
       sections [{type=MovieSceneParameterSection, channels [Thumbnail Opacity {type=MovieSceneScriptingFloatChannel, keys [[0,0],[12,1]]}]}]
      }]
    }]
   display_rate {numerator=60}
  }
 }
 Anim_OnTouchSelected {
  binding_count=1
  length_seconds=0.25
  object=&self:Anim_OnTouchSelected
  sequence {
   bindings [Bg {
     id=#3
     possessed_class=Image
     tracks [{
       type=MovieSceneWidgetMaterialTrack
       sections [{
         type=MovieSceneParameterSection
         channels [
          Background Color BR {type=MovieSceneScriptingFloatChannel, keys [[0,0.1022],[25,0.0212]]}
          Background Color BG {type=MovieSceneScriptingFloatChannel, keys [[0,0.0232],[25,0]]}
          Background Color BB {type=MovieSceneScriptingFloatChannel, keys [[0,0.3916],[25,0.1119]]}
          Background Color BA {type=MovieSceneScriptingFloatChannel, const [1,2]}
         ]
        }]
      }]
    }]
   display_rate {numerator=100}
  }
 }
]
graphs [
 EventGraph {nodes [
   n1 = "On Clicked (WBP_CTA_ButtonPrimary)"!() -> (OutputDelegate, then, Button:Common Button Base Object Reference) {is=title K2Node_ComponentBoundEvent}
   HandleCTAButtonClicked(execute=[@n1.then, @n30.else], self:Fort Activity Browser Play With Friends Tile Object Reference)
   n2 = View in Sidebar!() -> (OutputDelegate, then, InteractionNameID:Name) {is=title K2Node_CustomEvent}
   > OpenSidebar(self:Fort Activity Browser Play With Friends Tile Object Reference)
   > CloseContextMenu(self:WBP Discover Tile Base Object Reference=Get WBP_Discover_Tile_Base() -> (WBP_Discover_Tile_Base:WBP Discover Tile Base Object Reference))
   AddContextMenuItems() {is=title K2Node_CustomEvent}
   > AddInteraction!(
    execute
    self:UIKit Context Menu Action Context Object Reference=Get ContextMenuActionContext!(self:WBP Discover Tile Base Object Reference=Get WBP_Discover_Tile_Base()) -> (ContextMenuActionContext:UIKit Context Menu Action Context Object Reference)
    InActionInfo_DisplayName:Text="LOCTEXT(\"View in Sidebar\")"
    InActionInfo_IconTexture:Texture 2D Soft Object Reference
    InActionInfo_IconMaterial:Material Instance Soft Object Reference
    InActionInfo_SortPriority:int=1000
    InActionInfo_bIsAvailable:bool=True
    InActionInfo_bIsEnabled:bool=True
    InActionInfo_bIsSelected:bool=False
    InActionInfo_bAddDivider:bool=False
    InActionInfo_bPositionDividerAfter:bool=False
    InActionInfo_bCloseMenuOnSelect:bool=False
    InActionInfo_InteractionNameID:Name=None
    InActionInfo_Interaction:Data Table Row Handle Structure=(DataTable=None,RowName="")
    InActionInfo_PressedDelegate:Delegate="(null).None"
    InActionInfo_ReleasedDelegate:Delegate="(null).None"
    InInteractionDelegate:"Delegate (by ref)"=@n2
   ) -> (then)
   n3 = InputMethodChanged!() -> (OutputDelegate, then, bNewInputType:ECommonInputType Enum) {is=title K2Node_CustomEvent}
   > n4 = Set Current Input Type(Current Input Type:ECommonInputType Enum=@n3.bNewInputType) -> (Output_Get:ECommonInputType Enum)
   > Set CTAButton Padding!(execute, self, InputType:ECommonInputType Enum=@n4) -> (then) {is=title K2Node_CallFunction}
   n5 = AND Boolean(A="Equal (Enum)"(A=@n4, B=Touch), B=IsMobileGame())
   > n6 = Branch(Condition=@n5)
   n7 = Get WBP_CTA_ButtonPrimary() -> (WBP_CTA_ButtonPrimary:WBP UIKit Button Regular Object Reference)
   > n8 = SetVisibility(self=@n7, InVisibility=Collapsed)
   Set Tile Size!(
    execute=[@n9.then, @n8.then]
    self:WBP Discover Tile Base Object Reference=Get WBP_Discover_Tile_Base()
    InSize:Vector 2D Structure=Select(Option 0:Vector 2D Structure=Get TileSize() -> (TileSize:Vector 2D Structure), Option 1:Vector 2D Structure=(X=810,Y=210), Index=@n5) -> (ReturnValue:Vector 2D Structure)
   ) -> (then) {is=title K2Node_CallFunction}
   > SetHeightOverride(self=Get SizeBox_Icons() -> (SizeBox_Icons:Size Box Object Reference), InHeightOverride=SelectFloat(A=64, B=48, bPickA=@n5))
   > SetSize!(
    execute
    self:Spacer Object Reference=Get Spacer_72() -> (Spacer_72:Spacer Object Reference)
    InSize:Vector 2D Structure=Select(Option 0=(X=1,Y=2), Option 1=(X=1,Y=5), Index=@n5)
   ) -> (then)
   n9 = SetVisibility(execute=@n6.else, self=@n7, InVisibility=SelfHitTestInvisible)
   Event PreConstruct()
   > SetBrushFromMaterial(self=Get Bg() -> (Bg:Image Object Reference), Material=Get BackgroundInstance() -> (BackgroundInstance:Material Instance Object Reference))
   > n10 = GetDynamicMaterial(self=Get Bg())
   > Set BackgroundDynamicMaterialInstance(BackgroundDynamicMaterialInstance:Material Instance Dynamic Object Reference=@n10) -> (Output_Get:Material Instance Dynamic Object Reference)
   > SetColorAndOpacity(
    self=[
     Get Text_SingleUserName() -> (Text_SingleUserName:Common Text Object Reference)
     Get Text_RichPresence() -> (Text_RichPresence:Common Text Object Reference)
    ]
    InColorAndOpacity=Make Slate Color(SpecifiedColor=Get ColorSharedTextAndImages() -> (ColorSharedTextAndImages:Linear Color Structure))
   )
   Event Construct()
   n11 = Create Event()
   n12 = Get CommonInputSubsystem()
   > Bind Event to On Input Method Changed(self=@n12, Delegate=@n11)
   > AddContextMenuItems() {is=title K2Node_CallFunction}
   > Set Right Left Background() {is=title K2Node_CallFunction}
   > SetRenderOpacity(self=Get WBP_CTA_ButtonPrimary(), InOpacity=0)
   > SetThumbnail!(execute, self, Texture:Texture Object Reference, IsLoaded:bool=false) -> (then)
   n13 = AND Boolean(A=IsMobileGame(), B=IsUsingTouch(PlayerController=GetOwningPlayer()))
   > n14 = Branch(Condition=@n13)
   > n15 = SetVisibility(self=Get WBP_CTA_ButtonPrimary(), InVisibility=Collapsed)
   Set Tile Size(
    execute=[@n15.then, @n14.else]
    self=Get WBP_Discover_Tile_Base()
    InSize=Select(Option 0=Get TileSize(), Option 1=(X=810,Y=210), Index=@n13)
   )
   > SetHeightOverride(self=Get SizeBox_Icons(), InHeightOverride=SelectFloat(A=64, B=48, bPickA=@n13))
   > SetSize(self=Get Spacer_72(), InSize=Select(Option 0=(X=1,Y=2), Option 1=(X=1,Y=5), Index=@n13))
   > Set CTAButton Padding(InputType=GetCurrentInputType(self=Get CommonInputSubsystem()))
   > SetIsInteractionEnabled!(execute, self:Common Button Base Object Reference=Get WBP_CTA_ButtonPrimary(), bInIsInteractionEnabled:bool=false) -> (then)
   n16 = Event UpdateIslandThumbnail!() -> (OutputDelegate, then, ThumbnailTexture:Texture Object Reference) {is=title K2Node_Event}
   > SetThumbnail(Texture=@n16.ThumbnailTexture, IsLoaded=true)
   n17 = Event OnTextureLoadingComplete!() -> (OutputDelegate, then, ThumbnailTexture:Texture Object Reference) {is=title K2Node_Event}
   > SetThumbnail(Texture=@n17.ThumbnailTexture, IsLoaded=true)
   n18 = Event UpdateOtherPlayersSubText!() -> (OutputDelegate, then, OtherPlayersSubText:"Text (by ref)") {is=title K2Node_Event}
   > SetVisibility(
    self=Get Text_JoinableParty() -> (Text_JoinableParty:Common Text Object Reference)
    InVisibility=Select(
     Option 0:ESlateVisibility Enum=Collapsed
     Option 1:ESlateVisibility Enum=HitTestInvisible
     Index=TextIsEmpty(InText=@n18.OtherPlayersSubText)
    ) -> (ReturnValue:ESlateVisibility Enum)
   )
   > "SetText (Text)"(self=Get Text_OtherPlayers() -> (Text_OtherPlayers:Common Text Object Reference), InText=@n18.OtherPlayersSubText)
   n19 = Event UpdateRichPresence!() -> (OutputDelegate, then, RichPresence:"Text (by ref)") {is=title K2Node_Event}
   > "SetText (Text)"(self=Get Text_RichPresence(), InText=@n19.RichPresence)
   n20 = Event UpdatePartyMemberNames!() -> (OutputDelegate, then, Names:"Text (by ref)") {is=title K2Node_Event}
   n21 = TextIsEmpty(InText=@n20.Names)
   > SetVisibility(
    self=Get RichText_DisplayNamesOrSingleFriendName() -> (RichText_DisplayNamesOrSingleFriendName:Common Rich Text Block Object Reference)
    InVisibility=Select(Option 0=HitTestInvisible, Option 1=Collapsed, Index=@n21)
   )
   > n22 = Branch(Condition=@n21)
   SetText(execute=@n22.else, self=Get RichText_DisplayNamesOrSingleFriendName(), InText=@n20.Names)
   n23 = Event OnTileActiveChanged!() -> (OutputDelegate, then, bIsTileActive:bool) {is=title K2Node_Event}
   > n24 = Get!(execute:Exec) -> (then:Exec, else:Exec, FortNestedButtonGroupSelectionVM:Fort Nested Button Group Selection VM Object Reference)
   > SetIsSelected!(
    execute
    self:Fort Nested Button Group Selection VM Object Reference=@n24
    InButton:Fort Nested Button Group VMItem Interface=Self-Reference()
    bInSelected:bool=@n23.bIsTileActive
   ) -> (then)
   n25 = Branch(execute=@n24.else, Condition=@n23.bIsTileActive)
   n26 = Select(
    Option 0:Widget Animation Object Reference=Get Anim_OnHoverSelect() -> (Anim_OnHoverSelect:Widget Animation Object Reference)
    Option 1:Widget Animation Object Reference=Get Anim_OnTouchSelected() -> (Anim_OnTouchSelected:Widget Animation Object Reference)
    Index=IsUsingTouch(PlayerController=GetOwningPlayer())
   ) -> (ReturnValue:Widget Animation Object Reference)
   > PlayAnimationForward(InAnimation=@n26)
   n27 = Get WBP_Discover_Tile_Base()
   > OnTileBaseSelectionActive(self:WBP Discover Tile Base Object Reference=@n27)
   PlayAnimationReverse(execute=@n25.else, InAnimation=@n26, PlaybackSpeed=4)
   > OnTileBaseSelectionNotActive(self:WBP Discover Tile Base Object Reference=@n27)
   n28 = Event UpdateCTAButtonInfo!() -> (OutputDelegate, then, InteractionText:"Text (by ref)", bInteractionFound:bool) {is=title K2Node_Event}
   > SetText!(execute, self:Fort CTAButton Object Reference=Get WBP_CTA_ButtonPrimary(), ButtonText:"Text (by ref)"=@n28.InteractionText) -> (then)
   n29 = Event UpdateCTAButtonInteraction!() -> (OutputDelegate, then, bIsInteractionEnabled:bool) {is=title K2Node_Event}
   > SetIsInteractionEnabled(self=Get WBP_CTA_ButtonPrimary(), bInIsInteractionEnabled=@n29.bIsInteractionEnabled)
   Event On Clicked()
   > n30 = Branch(Condition=IsUsingTouch(PlayerController=GetOwningPlayer()))
   Event Destruct()
   > Unbind Event from On Input Method Changed(self=@n12, Delegate=@n11)
   n31 = Event UpdateSingleFriendName!() -> (OutputDelegate, then, SingleFriendName:"Text (by ref)") {is=title K2Node_Event}
   > "SetText (Text)"(self=Get Text_SingleUserName(), InText=@n31.SingleFriendName)
   n32 = Event UpdateLastInteraction!() -> (OutputDelegate, then, LastInteraction:"Text (by ref)") {is=title K2Node_Event}
   n33 = TextIsEmpty(InText=@n32.LastInteraction)
   > SetVisibility(
    self=Get Text_TimeAgoPartied() -> (Text_TimeAgoPartied:Common Text Object Reference)
    InVisibility=Select(Option 0:ESlateVisibility Enum=HitTestInvisible, Option 1:ESlateVisibility Enum=Collapsed, Index=@n33) -> (ReturnValue:ESlateVisibility Enum)
   )
   > n34 = Branch(Condition=@n33)
   "SetText (Text)"(execute=@n34.else, self=Get Text_TimeAgoPartied(), InText=@n32.LastInteraction)
   > SetInviteVisuals()
   n35 = Event SetNestedButtonGroupSelectionVM!() -> (OutputDelegate, then, InVM:Fort Nested Button Group Selection VM Object Reference) {is=title K2Node_Event}
   > Set FortNestedButtonGroupSelectionVM(FortNestedButtonGroupSelectionVM:Fort Nested Button Group Selection VM Object Reference=@n35.InVM) -> (Output_Get:Fort Nested Button Group Selection VM Object Reference)
   n36 = Event BP_OnSelectionChanged!() -> (OutputDelegate, then, bIsSelected:bool) {is=title K2Node_Event}
   > n37 = Get()
   > SetIsSelected(self=@n37, InButton=Self-Reference(), bInSelected=@n36.bIsSelected)
   n38 = Branch(execute=@n37.else, Condition=@n36.bIsSelected)
   n39 = Get Anim_OnTouchSelected()
   > PlayAnimation(InAnimation=@n39)
   PlayAnimation(execute=@n38.else, InAnimation=@n39, PlayMode=Reverse)
   n40 = Event Party Information Updated!() -> (OutputDelegate, then, bInIsTileSelected:bool) {is=title K2Node_Event}
   > Set Single Player Visuals() {is=title K2Node_CallFunction}
   > SetInviteVisuals()
   > SetPrivacyVisuals()
   n41 = Select(
    Option 0:Widget Animation Object Reference=Get Anim_OnHoverSelect()
    Option 1:Widget Animation Object Reference=Get Anim_OnTouchSelected()
    Index=IsUsingTouch(PlayerController=GetOwningPlayer())
   ) -> (ReturnValue:Widget Animation Object Reference)
   > SetAnimationCurrentTime!(
    execute
    self:User Widget Object Reference
    InAnimation:Widget Animation Object Reference=@n41
    InTime:float=Select(
     Option 0:float=GetStartTime!(self:Widget Animation Object Reference=@n41) -> (ReturnValue:float)
     Option 1:float=GetEndTime(self=@n41)
     Index=@n40.bInIsTileSelected
    ) -> (ReturnValue:float)
   ) -> (then)
   > Set CTAButton Padding(InputType=GetCurrentInputType(self=Get CommonInputSubsystem()))
   > SetCCUCount!(execute, self, CCUCount:int=Get CurrentPartySize() -> (CurrentPartySize:int=0)) -> (then)
   Event OnTextureBeginLoading() {is=title K2Node_Event}
   > SetThumbnail(IsLoaded=false)
   Event On Entry Released() {is=title K2Node_Event}
   > Set CTAButton Padding(InputType=GetCurrentInputType(self=Get CommonInputSubsystem()))
   > SetRenderOpacity(self=Get WBP_CTA_ButtonPrimary(), InOpacity=0)
   > CleanupContextMenuOnRelease(self:WBP Discover Tile Base Object Reference=Get WBP_Discover_Tile_Base())
   n42 = Event OnGameActivityChanged!() -> (OutputDelegate, then, GameActivity:Fort Game Activity Object Reference) {is=title K2Node_Event}
   > n43 = Is Valid(InputObject=@n42.GameActivity)
   SetVisibility(
    execute=@n43.Is Not Valid
    self=Get SqueegeeSlot_Tag() -> (SqueegeeSlot_Tag:Squeegee Injection Slot Object Reference)
    InVisibility=Collapsed
   )
   ShowActivityData!(
    execute=@n43.Is Valid
    self:Fort Squeegee Widget Injection Slot Object Reference=Get SqueegeeSlot_Tag()
    GameActivity:Fort Game Activity Object Reference=@n42.GameActivity
   ) -> (then)
   > SetVisibility(self=Get SqueegeeSlot_Tag(), InVisibility=SelfHitTestInvisible)
  ]}
 GetNestedButtonGroupSelectionVM {nodes [
   GetNestedButtonGroupSelectionVM() {is=title K2Node_FunctionEntry}
   > Return Node!(
    execute
    ReturnValue:Fort Nested Button Group Selection VM Object Reference=Get FortNestedButtonGroupSelectionVM() -> (FortNestedButtonGroupSelectionVM:Fort Nested Button Group Selection VM Object Reference)
   )
  ]}
 Is Arabic Language {nodes [
   Is Arabic Language() {is=title K2Node_FunctionEntry}
   > Return Node!(execute, IsArabic:bool="Equal Exactly (String)"(A=GetCurrentCulture!() -> (ReturnValue:String), B=ar))
  ]}
 OnMouseButtonDown {nodes [
   n1 = OnMouseButtonDown!() -> (then, MyGeometry:Geometry Structure, MouseEvent:"Pointer Event Structure (by ref)") {is=title K2Node_FunctionEntry}
   > n2 = OnRightClickMouse!(execute, self:WBP Discover Tile Base Object Reference=Get WBP_Discover_Tile_Base(), MouseEvent:Pointer Event Structure=@n1.MouseEvent) -> (then, WasHandled:Event Reply Structure)
   > Return Node!(execute, ReturnValue:Event Reply Structure=@n2)
  ]}
 OnSelectedButtonChanged {nodes [
   n1 = OnSelectedButtonChanged!() -> (then, SelectedActivity:Fort Nested Button Group VMItem Interface) {is=title K2Node_FunctionEntry}
   > n2 = Branch(Condition="Equal (Object)"(A:Fort Nested Button Group VMItem Interface=@n1, B=Self-Reference()))
   > n3 = Branch(Condition=Get IsTileActive!(self:WBP Discover Tile Base Object Reference=Get WBP_Discover_Tile_Base()) -> (IsTileActive:bool))
   n4 = Select(
    Option 0:Widget Animation Object Reference=Get Anim_OnHoverSelect()
    Option 1:Widget Animation Object Reference=Get Anim_OnTouchSelected()
    Index=IsUsingTouch(PlayerController=GetOwningPlayer())
   ) -> (ReturnValue:Widget Animation Object Reference)
   PlayAnimation(execute=@n3.else, InAnimation=@n4)
   n5 = Get WBP_Discover_Tile_Base()
   > OnTileBaseSelectionActive(self=@n5)
   Branch(execute=@n2.else, Condition=Get IsTileActive(self=Get WBP_Discover_Tile_Base()))
   > PlayAnimation(InAnimation=@n4, PlayMode=Reverse)
   > OnTileBaseSelectionNotActive(self=@n5)
   > Branch(Condition=AND Boolean(A=GetSelected!(self:Common Button Base Object Reference) -> (ReturnValue:bool), B=Get bSelectable() -> (bSelectable:bool)))
   > SetSelectedInternal!(execute, self:Common Button Base Object Reference, bInSelected:bool=false, bAllowSound:bool=false, bBroadcast:bool=true) -> (then)
  ]}
 Set CTAButton Padding {nodes [
   n1 = Set CTAButton Padding!() -> (then, InputType:ECommonInputType Enum) {is=title K2Node_FunctionEntry}
   n2 = Get WBP_CTA_ButtonPrimary()
   n3 = Break Margin!(Margin:"Margin Structure (by ref)"=Get Padding!(self:Overlay Slot Object Reference=SlotAsOverlaySlot(Widget=@n2)) -> (Padding:Margin Structure)) -> (Left:float=~, Top:float=~, Right:float=~, Bottom:float=~) {is=title K2Node_BreakStruct}
   n4 = GetIsEnabled!(self:Widget Object Reference=Get WBP_CTA_ButtonPrimary()) -> (ReturnValue:bool)
   n5 = SelectFloat(A=24, B=0, bPickA=@n4)
   n6 = SelectFloat(A=24, B=0, bPickA=@n4)
   > SetPadding!(
    execute
    self:Overlay Slot Object Reference=SlotAsOverlaySlot(Widget=Get WBP_CTA_ButtonPrimary())
    InPadding:Margin Structure=Select!(
     MouseAndKeyboard:Margin Structure=Make Margin(Left=@n6, Top=@n3.Top, Right=@n3.Right, Bottom=@n3.Bottom)
     Gamepad:Margin Structure=Make Margin(Left=@n5, Top=@n3.Top, Right=@n3.Right, Bottom=@n3.Bottom)
     Touch:Margin Structure
     Count:Margin Structure
     Index:ECommonInputType Enum=@n1
    ) -> (ReturnValue:Margin Structure)
   ) -> (then)
  ]}
 Set Right Left Background {nodes [
   Set Right Left Background() {is=title K2Node_FunctionEntry}
   > n1 = GetDynamicMaterial(self=Get Bg())
   > SetScalarParameterValue(
    self=@n1
    ParameterName=Mask Direction
    Value=Select!(Option 0:float=0.5, Option 1:float=-0.5, Index:bool=Is Arabic Language!() -> (IsArabic:bool) {is=title K2Node_CallFunction}) -> (ReturnValue:float)
   )
  ]}
 Set Single Player Visuals {nodes [
   Set Single Player Visuals() {is=title K2Node_FunctionEntry}
   n1 = "integer > integer"(A=Get CurrentPartySize() -> (CurrentPartySize=0), B=1)
   > SetVisibility(
    self=Get Text_SingleUserName()
    InVisibility=Select(Option 0:ESlateVisibility Enum=HitTestInvisible, Option 1:ESlateVisibility Enum=Collapsed, Index=@n1) -> (ReturnValue:ESlateVisibility Enum)
   )
   > SetVisibility(self=Get Text_OtherPlayers(), InVisibility=Select(Option 0=Collapsed, Option 1=HitTestInvisible, Index=@n1))
   > SetVisibility(self=Get Text_TimeAgoPartied(), InVisibility=Select(Option 0=SelfHitTestInvisible, Option 1=Collapsed, Index=@n1))
  ]}
 SetCCUCount {nodes [
   n1 = SetCCUCount!() -> (then, CCUCount:int) {is=title K2Node_FunctionEntry}
   > Update Secondary Text as CCU!(execute, self:WBP Discover Tile Base Object Reference=Get WBP_Discover_Tile_Base(), CCUCount:int=@n1) -> (then)
  ]}
 SetInviteVisuals {nodes [
   SetInviteVisuals() {is=title K2Node_FunctionEntry}
   > SetScalarParameterValue(
    self=Get BackgroundDynamicMaterialInstance() -> (BackgroundDynamicMaterialInstance:Material Instance Dynamic Object Reference)
    ParameterName=Get FlareOpacity() -> (FlareOpacity:Name=None)
    Value=Select(Option 0:float=0, Option 1:float=1, Index=Get bIsActiveInvite() -> (bIsActiveInvite:bool)) -> (ReturnValue:float)
   )
   > SetVisibility(
    self=Get Text_InviteToParty() -> (Text_InviteToParty:Common Text Object Reference)
    InVisibility=Select(Option 0:ESlateVisibility Enum=Collapsed, Option 1:ESlateVisibility Enum=SelfHitTestInvisible, Index=Get bIsActiveInvite()) -> (ReturnValue:ESlateVisibility Enum)
   )
   > Branch(Condition=Get bIsActiveInvite())
   > SetVisibility(self=[Get Text_JoinableParty(), Get Text_OtherPlayers(), Get Text_TimeAgoPartied()], InVisibility=Collapsed)
  ]}
 SetPrivacyVisuals {nodes [
   SetPrivacyVisuals() {is=title K2Node_FunctionEntry}
   > SetVisibility(
    self=Get WBP_JoinableParty_State_Private() -> (WBP_JoinableParty_State_Private:WBP Joinable Party State Private Object Reference)
    InVisibility=Select(Option 0=Collapsed, Option 1=HitTestInvisible, Index=Get bIsPartyPrivate() -> (bIsPartyPrivate:bool))
   )
  ]}
 SetThumbnail {nodes [
   n1 = SetThumbnail!() -> (then, Texture:Texture Object Reference, IsLoaded:bool=~) {is=title K2Node_FunctionEntry}
   > n2 = Branch(Condition=@n1.IsLoaded)
   n3 = Get BackgroundDynamicMaterialInstance()
   > SetTextureParameterValue(self=@n3, ParameterName=Get ThumbnailTexture() -> (ThumbnailTexture:Name=None), Value=@n1)
   > PlayAnimationForward(InAnimation=Get Anim_OnImageLoaded() -> (Anim_OnImageLoaded:Widget Animation Object Reference))
   SetScalarParameterValue(execute=@n2.else, self=@n3, ParameterName=Get ThumbnailOpacity() -> (ThumbnailOpacity:Name=None), Value=0)
  ]}
]
variables [
 BackgroundDynamicMaterialInstance {type=MaterialInstanceDynamic}
 BackgroundInstance {type=MaterialInstance, default=/DiscoveryBrowser/Tiles/JoinableParty/MaterialInstances/MI_UI_JoinableParty_Tile_Bg}
 ColorSharedTextAndImages {type=LinearColor, default=<rgba 0.5635 0.5875 0.6458 1>}
 Current Input Type {type=ECommonInputType, default=MouseAndKeyboard}
 FlareOpacity {type=string, default=Flare Opacity}
 HorizontalBoxContent {type="array<Widget>", default []}
 ThumbnailOpacity {type=string, default=Thumbnail Opacity}
 ThumbnailTexture {type=string, default=Thumbnail}
 TileSize {type=Vector2D, default=<648 196>}
]
widgets {
 info {
  inheritedWidgetCount=2
  namedSlotCount=2
  parentClass=/Script/DiscoveryBrowserUI.FortActivityBrowserPlayWithFriendsTile
  rootWidgetClass=&Overlay
  widgetCount=26
 }
 widgets [
  Overlay_Root {widgetClassPath=&Overlay, widget=Overlay_Root}
  WBP_Discover_Tile_Base {
   widgetClassPath=/DiscoveryBrowser/Tiles/WBP_Discover_Tile_Base
   bIsVariable=true
   parent=Overlay_Root
   widget=WBP_Discover_Tile_Base
   properties {isJoinablePartyTile=true, useIconBlock=true, useSecondaryTextBlock=true, tileSize=<648 196>}
   slot {class=OverlaySlot}
  }
  GridPanel_Root {
   widgetClassPath=&GridPanel
   namedSlotHost=WBP_Discover_Tile_Base
   widget=GridPanel_Root
   properties {clipping=ClipToBounds, columnFill [1], rowFill [0,1]}
  }
  UIKit_Backplate {
   widgetClassPath=&WBP_UIKit_Backplate
   parent=GridPanel_Root
   widget=UIKit_Backplate
   properties {brightness=Dark, cornerRadius=Interactive Component}
   slot {class=GridSlot, properties {rowSpan=2}}
  }
  Bg {
   widgetClassPath=&Image
   bIsVariable=true
   parent=GridPanel_Root
   widget=Bg
   properties {
    visibility=SelfHitTestInvisible
    brush {
     bIsDynamicallyLoaded=false
     drawAs=Box
     imageType=NoImage
     mirroring=NoMirror
     resourceName=None
     resourceObject=&MI_UI_JoinableParty_Tile_Bg
     tiling=NoTile
     imageSize=<32 32>
     margin=<ltrb 0.5 0.5 0.5 0.5>
     outlineSettings {
      bUseBrushTransparency=false
      roundingType=HalfHeightRadius
      width=0
      color {colorUseRule=UseColor_Specified, specifiedColor=<rgba 0 0 0 0>}
      cornerRadii=<0 0 0 0>
     }
     tintColor {colorUseRule=UseColor_Specified, specifiedColor=<rgba 1 1 1 1>}
     uVRegion {bIsValid=false, max=<0 0>, min=<0 0>}
    }
   }
   slot {class=GridSlot, properties {layer=1, rowSpan=2}}
  }
  WBP_JoinableParty_State_Private {
   widgetClassPath=&/WBP_JoinableParty_State_Private
   bIsVariable=true
   parent=GridPanel_Root
   widget=WBP_JoinableParty_State_Private
   properties {visibility=Collapsed}
   slot {class=GridSlot, properties {horizontalAlignment=HAlign_Right, layer=2, verticalAlignment=VAlign_Top, padding=<ltrb 0 12 12 0>}}
  }
  Grid_TopInfo {
   widgetClassPath=&VerticalBox
   parent=GridPanel_Root
   widget=Grid_TopInfo
   slot {class=GridSlot, properties {layer=2, padding=<ltrb 24 16 24 0>}}
  }
  SizeBox_Icons {
   widgetClassPath=&SizeBox
   bIsVariable=true
   parent=Grid_TopInfo
   widget=SizeBox_Icons
   properties {bOverride_HeightOverride=true, heightOverride=48}
   slot {class=VerticalBoxSlot, properties {horizontalAlignment=HAlign_Left, verticalAlignment=VAlign_Top}}
  }
  ScaleBox_0 {widgetClassPath=&ScaleBox, parent=SizeBox_Icons, widget=ScaleBox_0, properties {stretch=ScaleToFitY}, slot {class=SizeBoxSlot}}
  PartyMembersAvatarsDisplay {
   widgetClassPath=/Game/UI/Sidebar/SidebarJoinablePartyPortraits
   bInherited=true
   bIsVariable=true
   parent=ScaleBox_0
   widget=PartyMembersAvatarsDisplay
   properties {maxDisplayedUsers=3}
   slot {class=ScaleBoxSlot}
  }
  Spacer_72 {
   widgetClassPath=&Spacer
   bIsVariable=true
   parent=Grid_TopInfo
   widget=Spacer_72
   properties {size=<1 2>}
   slot {class=VerticalBoxSlot}
  }
  HB_Names {
   widgetClassPath=&HorizontalBox
   bIsVariable=true
   parent=Grid_TopInfo
   widget=HB_Names
   slot {class=VerticalBoxSlot, properties {horizontalAlignment=HAlign_Left}}
  }
  RichText_DisplayNamesOrSingleFriendName {
   widgetClassPath=&CommonRichTextBlock
   bIsVariable=true
   parent=HB_Names
   widget=RichText_DisplayNamesOrSingleFriendName
   properties {
    bOverrideDefaultStyle=true
    bTintInlineIcon=true
    mobileTextBlockScale=1.5
    text="{PH} Top Friend Name"
    defaultTextStyleOverride {
     overflowPolicy=Clip
     transformPolicy=None
     colorAndOpacity {colorUseRule=UseColor_Specified, specifiedColor=<rgba 1 1 1 1>}
     font {
      bForceMonospaced=false
      bMaterialIsStencil=false
      fontMaterial=null
      fontObject=&HeadingNow
      letterSpacing=0
      monospacedWidth=1
      size=18
      skewAmount=0
      typefaceFontName=75 Medium
      outlineSettings {
       bApplyOutlineToDropShadows=false
       bMiteredCorners=false
       bSeparateFillAlpha=false
       outlineMaterial=null
       outlineSize=0
       outlineColor=<rgba 0 0 0 1>
      }
     }
     highlightColor {colorUseRule=UseColor_Specified, specifiedColor=<rgba 0 0 0 1>}
     highlightShape {
      bIsDynamicallyLoaded=false
      drawAs=Image
      imageType=NoImage
      mirroring=NoMirror
      resourceName=None
      resourceObject=null
      tiling=NoTile
      imageSize=<32 32>
      margin=<ltrb 0 0 0 0>
      outlineSettings {
       bUseBrushTransparency=false
       roundingType=HalfHeightRadius
       width=0
       color {colorUseRule=UseColor_Specified, specifiedColor=<rgba 0 0 0 0>}
       cornerRadii=<0 0 0 0>
      }
      tintColor {colorUseRule=UseColor_Specified, specifiedColor=<rgba 1 1 1 1>}
      uVRegion {bIsValid=false, max=<0 0>, min=<0 0>}
     }
     selectedBackgroundColor {colorUseRule=UseColor_Foreground, specifiedColor=<rgba 1 0 1 1>}
     shadowColorAndOpacity=<rgba 0 0 0 1>
     shadowOffset=<0 0>
     strikeBrush {
      bIsDynamicallyLoaded=false
      drawAs=Image
      imageType=NoImage
      mirroring=NoMirror
      resourceName=None
      resourceObject=null
      tiling=NoTile
      imageSize=<32 32>
      margin=<ltrb 0 0 0 0>
      outlineSettings {
       bUseBrushTransparency=false
       roundingType=HalfHeightRadius
       width=0
       color {colorUseRule=UseColor_Specified, specifiedColor=<rgba 0 0 0 0>}
       cornerRadii=<0 0 0 0>
      }
      tintColor {colorUseRule=UseColor_Specified, specifiedColor=<rgba 1 1 1 1>}
      uVRegion {bIsValid=false, max=<0 0>, min=<0 0>}
     }
     underlineBrush {
      bIsDynamicallyLoaded=false
      drawAs=Image
      imageType=NoImage
      mirroring=NoMirror
      resourceName=None
      resourceObject=null
      tiling=NoTile
      imageSize=<32 32>
      margin=<ltrb 0 0 0 0>
      outlineSettings {
       bUseBrushTransparency=false
       roundingType=HalfHeightRadius
       width=0
       color {colorUseRule=UseColor_Specified, specifiedColor=<rgba 0 0 0 0>}
       cornerRadii=<0 0 0 0>
      }
      tintColor {colorUseRule=UseColor_Specified, specifiedColor=<rgba 1 1 1 1>}
      uVRegion {bIsValid=false, max=<0 0>, min=<0 0>}
     }
    }
   }
   slot {class=HorizontalBoxSlot, properties {padding=<ltrb 0 0 8 0>}}
  }
  Text_SingleUserName {
   widgetClassPath=&CommonTextBlock
   bIsVariable=true
   parent=HB_Names
   widget=Text_SingleUserName
   properties {
    mobileFontSizeMultiplier=1.35
    text="(PH) Single Name"
    colorAndOpacity {colorUseRule=UseColor_Specified, specifiedColor=<rgba 0.6038 1 1 1>}
    font {
     bForceMonospaced=false
     bMaterialIsStencil=false
     fontMaterial=null
     fontObject=&BurbankSmall
     letterSpacing=0
     monospacedWidth=1
     size=16
     skewAmount=0
     typefaceFontName=Bold
     outlineSettings {
      bApplyOutlineToDropShadows=false
      bMiteredCorners=false
      bSeparateFillAlpha=false
      outlineMaterial=null
      outlineSize=0
      outlineColor=<rgba 0 0 0 1>
     }
    }
   }
   slot {class=HorizontalBoxSlot, properties {verticalAlignment=VAlign_Center}}
  }
  HorizontalBox_0 {widgetClassPath=&HorizontalBox, bIsVariable=true, parent=Grid_TopInfo, widget=HorizontalBox_0, slot {class=VerticalBoxSlot}}
  Text_OtherPlayers {
   widgetClassPath=&CommonTextBlock
   bIsVariable=true
   parent=HorizontalBox_0
   widget=Text_OtherPlayers
   properties {
    mobileFontSizeMultiplier=1.75
    text="(PH) +2 others"
    visibility=Collapsed
    colorAndOpacity {colorUseRule=UseColor_Specified, specifiedColor=<rgba 0.0369 0.9823 0.227 1>}
    font {
     bForceMonospaced=false
     bMaterialIsStencil=false
     fontMaterial=null
     fontObject=&BurbankSmall
     letterSpacing=0
     monospacedWidth=1
     size=13
     skewAmount=0
     typefaceFontName=Bold
     outlineSettings {
      bApplyOutlineToDropShadows=false
      bMiteredCorners=false
      bSeparateFillAlpha=false
      outlineMaterial=null
      outlineSize=0
      outlineColor=<rgba 0 0 0 1>
     }
    }
    shadowOffset=<0 0>
    shapedTextOptions {bOverride_TextFlowDirection=true, bOverride_TextShapingMethod=false, textFlowDirection=Culture, textShapingMethod=Auto}
   }
   slot {class=HorizontalBoxSlot}
  }
  Text_JoinableParty {
   widgetClassPath=&CommonTextBlock
   bIsVariable=true
   parent=HorizontalBox_0
   widget=Text_JoinableParty
   properties {
    mobileFontSizeMultiplier=1.75
    text=In a joinable party
    visibility=Collapsed
    colorAndOpacity {colorUseRule=UseColor_Specified, specifiedColor=<rgba 0.0369 0.9823 0.227 1>}
    font {
     bForceMonospaced=false
     bMaterialIsStencil=false
     fontMaterial=null
     fontObject=&BurbankSmall
     letterSpacing=0
     monospacedWidth=1
     size=13
     skewAmount=0
     typefaceFontName=Bold
     outlineSettings {
      bApplyOutlineToDropShadows=false
      bMiteredCorners=false
      bSeparateFillAlpha=false
      outlineMaterial=null
      outlineSize=0
      outlineColor=<rgba 0 0 0 1>
     }
    }
    shadowOffset=<0 0>
    shapedTextOptions {bOverride_TextFlowDirection=true, bOverride_TextShapingMethod=false, textFlowDirection=Culture, textShapingMethod=Auto}
   }
   slot {class=HorizontalBoxSlot}
  }
  Text_TimeAgoPartied {
   widgetClassPath=&CommonTextBlock
   bIsVariable=true
   parent=HorizontalBox_0
   widget=Text_TimeAgoPartied
   properties {
    text="(PH) Partied 24h ago"
    visibility=Collapsed
    colorAndOpacity {colorUseRule=UseColor_Specified, specifiedColor=<rgba 0.0369 0.9823 0.227 1>}
    font {
     bForceMonospaced=false
     bMaterialIsStencil=false
     fontMaterial=null
     fontObject=&BurbankSmall
     letterSpacing=0
     monospacedWidth=1
     size=13
     skewAmount=0
     typefaceFontName=Bold
     outlineSettings {
      bApplyOutlineToDropShadows=false
      bMiteredCorners=false
      bSeparateFillAlpha=false
      outlineMaterial=null
      outlineSize=0
      outlineColor=<rgba 0 0 0 1>
     }
    }
    shadowOffset=<0 0>
    shapedTextOptions {bOverride_TextFlowDirection=true, bOverride_TextShapingMethod=false, textFlowDirection=Culture, textShapingMethod=Auto}
   }
   slot {class=HorizontalBoxSlot}
  }
  Text_InviteToParty {
   widgetClassPath=&CommonTextBlock
   bIsVariable=true
   parent=HorizontalBox_0
   widget=Text_InviteToParty
   properties {
    mobileFontSizeMultiplier=1.75
    text=has invited you to their party
    visibility=Collapsed
    colorAndOpacity {colorUseRule=UseColor_Specified, specifiedColor=<rgba 0.0369 0.9823 0.227 1>}
    font {
     bForceMonospaced=false
     bMaterialIsStencil=false
     fontMaterial=null
     fontObject=&BurbankSmall
     letterSpacing=0
     monospacedWidth=1
     size=13
     skewAmount=0
     typefaceFontName=Bold
     outlineSettings {
      bApplyOutlineToDropShadows=false
      bMiteredCorners=false
      bSeparateFillAlpha=false
      outlineMaterial=null
      outlineSize=0
      outlineColor=<rgba 0 0 0 1>
     }
    }
    shadowOffset=<0 0>
    shapedTextOptions {bOverride_TextFlowDirection=true, bOverride_TextShapingMethod=false, textFlowDirection=Culture, textShapingMethod=Auto}
   }
   slot {class=HorizontalBoxSlot}
  }
  Grid_BottomInfo_Clips {
   widgetClassPath=&Overlay
   bIsVariable=true
   parent=GridPanel_Root
   widget=Grid_BottomInfo_Clips
   slot {class=GridSlot, properties {layer=3, row=1, verticalAlignment=VAlign_Bottom}}
  }
  HB_PresenceTexts {
   widgetClassPath=&HorizontalBox
   bIsVariable=true
   parent=Grid_BottomInfo_Clips
   widget=HB_PresenceTexts
   properties {visibility=Visible}
   slot {class=OverlaySlot, properties {verticalAlignment=VAlign_Center, padding=<ltrb 24 16 0 16>}}
  }
  SizeBox_MaxWidthForLongPresenceToForceEllipses {
   widgetClassPath=&SizeBox
   bIsVariable=true
   parent=HB_PresenceTexts
   widget=SizeBox_MaxWidthForLongPresenceToForceEllipses
   properties {bOverride_MaxDesiredWidth=true, maxDesiredWidth=540}
   slot {class=HorizontalBoxSlot, properties {horizontalAlignment=HAlign_Left}}
  }
  Text_RichPresence {
   widgetClassPath=&CommonTextBlock
   bIsVariable=true
   parent=SizeBox_MaxWidthForLongPresenceToForceEllipses
   widget=Text_RichPresence
   properties {
    clipping=ClipToBounds
    mobileFontSizeMultiplier=1.2
    scrollStyle=&TextScrollStyle-Base_C
    text="(PH) Main Lobby"
    textOverflowPolicy=Ellipsis
    visibility=HitTestInvisible
    colorAndOpacity {colorUseRule=UseColor_Specified, specifiedColor=<rgba 0.6038 1 1 1>}
    font {
     bForceMonospaced=false
     bMaterialIsStencil=false
     fontMaterial=null
     fontObject=&BurbankSmall
     letterSpacing=0
     monospacedWidth=1
     size=15
     skewAmount=0
     typefaceFontName=Bold
     outlineSettings {
      bApplyOutlineToDropShadows=false
      bMiteredCorners=false
      bSeparateFillAlpha=false
      outlineMaterial=null
      outlineSize=0
      outlineColor=<rgba 0 0 0 1>
     }
    }
    shapedTextOptions {bOverride_TextFlowDirection=true, bOverride_TextShapingMethod=false, textFlowDirection=Culture, textShapingMethod=Auto}
   }
   slot {class=SizeBoxSlot}
  }
  WBP_CTA_ButtonPrimary {
   widgetClassPath=&WBP_UIKit_Button_Regular
   bIsVariable=true
   parent=Grid_BottomInfo_Clips
   widget=WBP_CTA_ButtonPrimary
   properties {buttonSize=Small, triggeringInputAction {dataTable=&InputActionDataTable, rowName=Accept}}
   slot {class=OverlaySlot, properties {verticalAlignment=VAlign_Center, padding=<ltrb 24 16 0 16>}}
  }
  SqueegeeSlot_Tag {
   widgetClassPath=/Game/UI/Plugin/Squeegee/SqueegeeInjectionSlot
   bIsVariable=true
   namedSlotHost=WBP_Discover_Tile_Base
   widget=SqueegeeSlot_Tag
   properties {
    pluginId=Squeegee
    pluginWidgetId=RatingTag
    styleParams {bShowDetails=true, fontSize=15, paddingSize=0.5, iconSize=<26 26>, iconSkewSize=<5 26>}
   }
  }
  InputActionWidget {widgetClassPath=&CommonActionWidget, bInherited=true, widget=null}
 ]
}
