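# Small rounded-corner Discovery browser tile for a joinable friend party, built on FortActivityBrowserPlayWithFriendsTile and hosted in WBP_Discover_Tile_Base slots.
# Parent events fill avatars, names, action text, party count and activity rating tag.
# Selection runs through a nested button group view model, or falls back to base-tile calls, driving a material glow animation and invite parameter.
# Non-touch clicks trigger the CTA; a context-menu action opens the sidebar.
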WidgetBlueprint /DiscoveryBrowser/Tiles/JoinableParty/WBP_ActivityBrowserTile_JoinableParty_RoundedCorners_Small
animations [Anim_OnHoverSelect {
  binding_count=1
  length_seconds=0.25
  object=&self:Anim_OnHoverSelect
  sequence {bindings [Bg {
     id=#1
     possessed_class=Image
     tracks [{
       type=MovieSceneWidgetMaterialTrack
       sections [{type=MovieSceneParameterSection, channels [Glow Opacity {type=MovieSceneScriptingFloatChannel, keys [[0.0003,0],[5,1]]}]}]
      }]
    }]}
 }]
graphs [
 EventGraph {nodes [
   n1 = View in Sidebar!() -> (OutputDelegate, then, InteractionNameID:Name) {is=title K2Node_CustomEvent}
   > OpenSidebar(self:Fort Activity Browser Play With Friends Tile Object Reference)
   > CloseContextMenu(self:WBP Discover Tile Base Object Reference=Get WBP_Discover_Tile_Base() -> (WBP_Discover_Tile_Base:WBP Discover Tile Base Object Reference))
   AddContextMenuItems() {is=title K2Node_CustomEvent}
   > AddInteraction!(
    execute
    self:UIKit Context Menu Action Context Object Reference=Get ContextMenuActionContext!(self:WBP Discover Tile Base Object Reference=Get WBP_Discover_Tile_Base()) -> (ContextMenuActionContext:UIKit Context Menu Action Context Object Reference)
    InActionInfo_DisplayName:Text="LOCTEXT(\"View in Sidebar\")"
    InActionInfo_IconTexture:Texture 2D Soft Object Reference
    InActionInfo_IconMaterial:Material Instance Soft Object Reference
    InActionInfo_SortPriority:int=1000
    InActionInfo_bIsAvailable:bool=True
    InActionInfo_bIsEnabled:bool=True
    InActionInfo_bIsSelected:bool=False
    InActionInfo_bAddDivider:bool=False
    InActionInfo_bPositionDividerAfter:bool=False
    InActionInfo_bCloseMenuOnSelect:bool=False
    InActionInfo_InteractionNameID:Name=None
    InActionInfo_Interaction:Data Table Row Handle Structure=(DataTable=None,RowName="")
    InActionInfo_PressedDelegate:Delegate="(null).None"
    InActionInfo_ReleasedDelegate:Delegate="(null).None"
    InInteractionDelegate:"Delegate (by ref)"=@n1
   ) -> (then)
   Event Construct()
   > Parent: Construct()
   > AddContextMenuItems() {is=title K2Node_CallFunction}
   n2 = Event UpdatePartyMemberNames!() -> (OutputDelegate, then, Names:"Text (by ref)") {is=title K2Node_Event}
   n3 = Event UpdateSingleFriendName!() -> (OutputDelegate, then, SingleFriendName:"Text (by ref)") {is=title K2Node_Event}
   n4 = Get CurrentPartySize() -> (CurrentPartySize:int=0)
   SetText(
    execute=[@n2.then, @n3.then]
    self=Get RichText_DisplayNamesOrSingleFriendName() -> (RichText_DisplayNamesOrSingleFriendName:Common Rich Text Block Object Reference)
    InText=Select(Option 0:"Text (by ref)"=@n3.SingleFriendName, Option 1:"Text (by ref)"=@n2.Names, Index="integer > integer"(A=@n4, B=1)) -> (ReturnValue:"Text (by ref)")
   )
   Event On Entry Released() {is=title K2Node_Event}
   > CleanupContextMenuOnRelease(self:WBP Discover Tile Base Object Reference=Get WBP_Discover_Tile_Base())
   n5 = Event OnTileActiveChanged!() -> (OutputDelegate, then, bIsTileActive:bool) {is=title K2Node_Event}
   > n6 = Get!(execute:Exec) -> (then:Exec, else:Exec, FortNestedButtonGroupSelectionVM:Fort Nested Button Group Selection VM Object Reference)
   > SetIsSelected!(
    execute
    self:Fort Nested Button Group Selection VM Object Reference=@n6
    InButton:Fort Nested Button Group VMItem Interface=Self-Reference()
    bInSelected:bool=@n5.bIsTileActive
   ) -> (then)
   n7 = Branch(execute=@n6.else, Condition=@n5.bIsTileActive)
   n8 = Get WBP_Discover_Tile_Base()
   > OnTileBaseSelectionActive(self:WBP Discover Tile Base Object Reference=@n8)
   n9 = Get Anim_OnHoverSelect() -> (Anim_OnHoverSelect:Widget Animation Object Reference)
   > PlayAnimationForward(InAnimation=@n9)
   OnTileBaseSelectionNotActive(execute=@n7.else, self:WBP Discover Tile Base Object Reference=@n8)
   > PlayAnimationReverse(InAnimation=@n9, PlaybackSpeed=4)
   n10 = Event UpdateCTAButtonInfo!() -> (OutputDelegate, then, InteractionText:"Text (by ref)", bInteractionFound:bool) {is=title K2Node_Event}
   > "SetText (Text)"(self=Get Text_RichPresence() -> (Text_RichPresence:Common Text Object Reference), InText=@n10.InteractionText)
   Event On Clicked()
   > n11 = Branch(Condition=IsUsingTouch(PlayerController=GetOwningPlayer()))
   HandleCTAButtonClicked(execute=@n11.else, self:Fort Activity Browser Play With Friends Tile Object Reference)
   n12 = Event SetNestedButtonGroupSelectionVM!() -> (OutputDelegate, then, InVM:Fort Nested Button Group Selection VM Object Reference) {is=title K2Node_Event}
   > Set FortNestedButtonGroupSelectionVM(FortNestedButtonGroupSelectionVM:Fort Nested Button Group Selection VM Object Reference=@n12.InVM) -> (Output_Get:Fort Nested Button Group Selection VM Object Reference)
   n13 = Event Party Information Updated!() -> (OutputDelegate, then, bInIsTileSelected:bool) {is=title K2Node_Event}
   > SetSinglePlayerVisuals()
   > SetInviteVisuals()
   n14 = Get Anim_OnHoverSelect()
   > SetAnimationCurrentTime!(
    execute
    self:User Widget Object Reference
    InAnimation:Widget Animation Object Reference=@n14
    InTime:float=Select(
     Option 0:float=GetStartTime!(self:Widget Animation Object Reference=@n14) -> (ReturnValue:float)
     Option 1:float=GetEndTime(self=@n14)
     Index=@n13.bInIsTileSelected
    ) -> (ReturnValue:float)
   ) -> (then)
   > SetCCUCount!(execute, self, CCUCount:int=Get CurrentPartySize() -> (CurrentPartySize=0)) -> (then)
   n15 = Event OnUpdateColumnSize!() -> (OutputDelegate, then, NewColumnSize:int=0) {is=title K2Node_Event}
   n16 = Get WBP_Discover_Tile_Base()
   > Set Tile Size!(
    execute
    self:WBP Discover Tile Base Object Reference=@n16
    InSize:Vector 2D Structure=ColumnSizeToTileSize!(self:WBP Discover Tile Base Object Reference=@n16, NewColumnSize:int=@n15.NewColumnSize, IsSquareTiles:bool=false) -> (NewTileSize:Vector 2D Structure)
   ) -> (then) {is=title K2Node_CallFunction}
   n17 = Event OnGameActivityChanged!() -> (OutputDelegate, then, GameActivity:Fort Game Activity Object Reference) {is=title K2Node_Event}
   > n18 = Is Valid(InputObject=@n17.GameActivity)
   SetVisibility(
    execute=@n18.Is Not Valid
    self=Get SqueegeeSlot_Tag() -> (SqueegeeSlot_Tag:Squeegee Injection Slot Object Reference)
    InVisibility=Collapsed
   )
   ShowActivityData!(
    execute=@n18.Is Valid
    self:Fort Squeegee Widget Injection Slot Object Reference=Get SqueegeeSlot_Tag()
    GameActivity:Fort Game Activity Object Reference=@n17.GameActivity
   ) -> (then)
   > SetVisibility(self=Get SqueegeeSlot_Tag(), InVisibility=SelfHitTestInvisible)
   Event PreConstruct()
   Event UpdateOtherPlayersSubText!() -> (OutputDelegate, then, OtherPlayersSubText:"Text (by ref)") {is=title K2Node_Event}
   Event UpdateRichPresence!() -> (OutputDelegate, then, RichPresence:"Text (by ref)") {is=title K2Node_Event}
   Event UpdateLastInteraction!() -> (OutputDelegate, then, LastInteraction:"Text (by ref)") {is=title K2Node_Event}
  ]}
 GetNestedButtonGroupSelectionVM {nodes [
   GetNestedButtonGroupSelectionVM() {is=title K2Node_FunctionEntry}
   > Return Node!(
    execute
    ReturnValue:Fort Nested Button Group Selection VM Object Reference=Get FortNestedButtonGroupSelectionVM() -> (FortNestedButtonGroupSelectionVM:Fort Nested Button Group Selection VM Object Reference)
   )
  ]}
 Is Arabic Language {nodes [
   Is Arabic Language() {is=title K2Node_FunctionEntry}
   > Return Node!(execute, IsArabic:bool="Equal Exactly (String)"(A=GetCurrentCulture!() -> (ReturnValue:String), B=ar))
  ]}
 OnMouseButtonDown {nodes [
   n1 = OnMouseButtonDown!() -> (then, MyGeometry:Geometry Structure, MouseEvent:"Pointer Event Structure (by ref)") {is=title K2Node_FunctionEntry}
   > n2 = OnRightClickMouse!(execute, self:WBP Discover Tile Base Object Reference=Get WBP_Discover_Tile_Base(), MouseEvent:Pointer Event Structure=@n1.MouseEvent) -> (then, WasHandled:Event Reply Structure)
   > Return Node!(execute, ReturnValue:Event Reply Structure=@n2)
  ]}
 OnSelectedButtonChanged {nodes [
   n1 = OnSelectedButtonChanged!() -> (then, SelectedActivity:Fort Nested Button Group VMItem Interface) {is=title K2Node_FunctionEntry}
   > n2 = Branch(Condition="Equal (Object)"(A:Fort Nested Button Group VMItem Interface=@n1, B=Self-Reference()))
   n3 = Get IsTileActive!(self:WBP Discover Tile Base Object Reference=Get WBP_Discover_Tile_Base()) -> (IsTileActive:bool)
   > n4 = Branch(Condition=@n3)
   n5 = Get WBP_Discover_Tile_Base()
   OnTileBaseSelectionActive(execute=@n4.else, self=@n5)
   n6 = Get Anim_OnHoverSelect()
   > PlayAnimationForward(InAnimation=@n6)
   Branch(execute=@n2.else, Condition=@n3)
   > OnTileBaseSelectionNotActive(self=@n5)
   > PlayAnimationReverse(InAnimation=@n6, PlaybackSpeed=4)
   > Branch(Condition=AND Boolean(A=GetSelected!(self:Common Button Base Object Reference) -> (ReturnValue:bool), B=Get bSelectable() -> (bSelectable:bool)))
   > SetSelectedInternal!(execute, self:Common Button Base Object Reference, bInSelected:bool=false, bAllowSound:bool=false, bBroadcast:bool=true) -> (then)
  ]}
 SetCCUCount {nodes [
   n1 = SetCCUCount!() -> (then, CCUCount:int) {is=title K2Node_FunctionEntry}
   > Update Secondary Text as CCU!(execute, self:WBP Discover Tile Base Object Reference=Get WBP_Discover_Tile_Base(), CCUCount:int=@n1) -> (then)
  ]}
 SetInviteVisuals {nodes [
   SetInviteVisuals() {is=title K2Node_FunctionEntry}
   > n1 = GetDynamicMaterial(self=Get Bg() -> (Bg:Image Object Reference))
   > SetScalarParameterValue(
    self=@n1
    ParameterName=Get AnimInvite() -> (AnimInvite:Name=None)
    Value=Select(Option 0=0, Option 1=1, Index=Get bIsActiveInvite() -> (bIsActiveInvite:bool))
   )
  ]}
 SetSinglePlayerVisuals {nodes [SetSinglePlayerVisuals() {is=title K2Node_FunctionEntry}]}
]
variables [AnimInvite {type=string, default=Anim_Invite}]
widgets {
 info {
  inheritedWidgetCount=2
  namedSlotCount=2
  parentClass=/Script/DiscoveryBrowserUI.FortActivityBrowserPlayWithFriendsTile
  rootWidgetClass=&Overlay
  widgetCount=16
 }
 widgets [
  Overlay_1 {widgetClassPath=&Overlay, widget=Overlay_1}
  WBP_Discover_Tile_Base {
   widgetClassPath=/DiscoveryBrowser/Tiles/WBP_Discover_Tile_Base
   bIsVariable=true
   parent=Overlay_1
   widget=WBP_Discover_Tile_Base
   properties {isJoinablePartyTile=true, useIconBlock=true, useSecondaryTextBlock=true, tileSize=<248 140>}
   slot {class=OverlaySlot}
  }
  GridPanel_Root {widgetClassPath=&GridPanel, namedSlotHost=WBP_Discover_Tile_Base, widget=GridPanel_Root, properties {columnFill [1], rowFill [1]}}
  UIKit_Backplate {
   widgetClassPath=&WBP_UIKit_Backplate
   parent=GridPanel_Root
   widget=UIKit_Backplate
   properties {brightness=Dark, cornerRadius=Interactive Component}
   slot {class=GridSlot, properties {rowSpan=2}}
  }
  Bg {
   widgetClassPath=&Image
   bIsVariable=true
   parent=GridPanel_Root
   widget=Bg
   properties {
    visibility=SelfHitTestInvisible
    brush {
     bIsDynamicallyLoaded=false
     drawAs=Box
     imageType=NoImage
     mirroring=NoMirror
     resourceName=None
     resourceObject=/DiscoveryBrowser/Tiles/JoinableParty/Materials/MI_UI_Party_Tile_HomeRow_Bg
     tiling=NoTile
     imageSize=<32 32>
     margin=<ltrb 0.5 0.5 0.5 0.5>
     outlineSettings {
      bUseBrushTransparency=false
      roundingType=HalfHeightRadius
      width=0
      color {colorUseRule=UseColor_Specified, specifiedColor=<rgba 0 0 0 0>}
      cornerRadii=<0 0 0 0>
     }
     tintColor {colorUseRule=UseColor_Specified, specifiedColor=<rgba 1 1 1 1>}
     uVRegion {bIsValid=false, max=<0 0>, min=<0 0>}
    }
   }
   slot {class=GridSlot, properties {layer=1, rowSpan=2}}
  }
  Grid_TopInfo {
   widgetClassPath=&VerticalBox
   bIsVariable=true
   parent=GridPanel_Root
   widget=Grid_TopInfo
   slot {class=GridSlot, properties {layer=5, rowSpan=2, padding=<ltrb 16 16 16 16>}}
  }
  SizeBox_Icons {
   widgetClassPath=&SizeBox
   bIsVariable=true
   parent=Grid_TopInfo
   widget=SizeBox_Icons
   properties {bOverride_HeightOverride=true, heightOverride=48}
   slot {class=VerticalBoxSlot, properties {horizontalAlignment=HAlign_Left, verticalAlignment=VAlign_Top}}
  }
  ScaleBox_0 {widgetClassPath=&ScaleBox, parent=SizeBox_Icons, widget=ScaleBox_0, properties {stretch=ScaleToFitY}, slot {class=SizeBoxSlot}}
  PartyMembersAvatarsDisplay {
   widgetClassPath=/Game/UI/Sidebar/SidebarJoinablePartyPortraits
   bInherited=true
   bIsVariable=true
   parent=ScaleBox_0
   widget=PartyMembersAvatarsDisplay
   properties {maxDisplayedUsers=3}
   slot {class=ScaleBoxSlot}
  }
  ContentSpacer {
   widgetClassPath=&Spacer
   bIsVariable=true
   parent=Grid_TopInfo
   widget=ContentSpacer
   properties {size=<1 16>}
   slot {class=VerticalBoxSlot, properties {size {sizeRule=Fill, value=1}}}
  }
  SizeBox_MaxWidthForLongPresenceToForceEllipses {
   widgetClassPath=&SizeBox
   bIsVariable=true
   parent=Grid_TopInfo
   widget=SizeBox_MaxWidthForLongPresenceToForceEllipses
   properties {bOverride_MaxDesiredWidth=true, maxDesiredWidth=540}
   slot {class=VerticalBoxSlot}
  }
  VerticalBox_0 {
   widgetClassPath=&VerticalBox
   parent=SizeBox_MaxWidthForLongPresenceToForceEllipses
   widget=VerticalBox_0
   slot {class=SizeBoxSlot}
  }
  RichText_DisplayNamesOrSingleFriendName {
   widgetClassPath=&CommonRichTextBlock
   bIsVariable=true
   parent=VerticalBox_0
   widget=RichText_DisplayNamesOrSingleFriendName
   properties {
    bOverrideDefaultStyle=true
    bTintInlineIcon=true
    mobileTextBlockScale=1.3
    scrollStyle=/FortUILibrary/Styles/Typography/Foundation/TextScrollStyle-NoFade.TextScrollStyle-NoFade_C
    text="{PH} Friend Name"
    defaultTextStyleOverride {
     overflowPolicy=Clip
     transformPolicy=None
     colorAndOpacity {colorUseRule=UseColor_Specified, specifiedColor=<rgba 1 1 1 1>}
     font {
      bForceMonospaced=false
      bMaterialIsStencil=false
      fontMaterial=null
      fontObject=&HeadingNow
      letterSpacing=0
      monospacedWidth=1
      size=18
      skewAmount=0
      typefaceFontName=75 Medium
      outlineSettings {
       bApplyOutlineToDropShadows=false
       bMiteredCorners=false
       bSeparateFillAlpha=false
       outlineMaterial=null
       outlineSize=0
       outlineColor=<rgba 0 0 0 1>
      }
     }
     highlightColor {colorUseRule=UseColor_Specified, specifiedColor=<rgba 0 0 0 1>}
     highlightShape {
      bIsDynamicallyLoaded=false
      drawAs=Image
      imageType=NoImage
      mirroring=NoMirror
      resourceName=None
      resourceObject=null
      tiling=NoTile
      imageSize=<32 32>
      margin=<ltrb 0 0 0 0>
      outlineSettings {
       bUseBrushTransparency=false
       roundingType=HalfHeightRadius
       width=0
       color {colorUseRule=UseColor_Specified, specifiedColor=<rgba 0 0 0 0>}
       cornerRadii=<0 0 0 0>
      }
      tintColor {colorUseRule=UseColor_Specified, specifiedColor=<rgba 1 1 1 1>}
      uVRegion {bIsValid=false, max=<0 0>, min=<0 0>}
     }
     selectedBackgroundColor {colorUseRule=UseColor_Foreground, specifiedColor=<rgba 1 0 1 1>}
     shadowColorAndOpacity=<rgba 0 0 0 1>
     shadowOffset=<0 0>
     strikeBrush {
      bIsDynamicallyLoaded=false
      drawAs=Image
      imageType=NoImage
      mirroring=NoMirror
      resourceName=None
      resourceObject=null
      tiling=NoTile
      imageSize=<32 32>
      margin=<ltrb 0 0 0 0>
      outlineSettings {
       bUseBrushTransparency=false
       roundingType=HalfHeightRadius
       width=0
       color {colorUseRule=UseColor_Specified, specifiedColor=<rgba 0 0 0 0>}
       cornerRadii=<0 0 0 0>
      }
      tintColor {colorUseRule=UseColor_Specified, specifiedColor=<rgba 1 1 1 1>}
      uVRegion {bIsValid=false, max=<0 0>, min=<0 0>}
     }
     underlineBrush {
      bIsDynamicallyLoaded=false
      drawAs=Image
      imageType=NoImage
      mirroring=NoMirror
      resourceName=None
      resourceObject=null
      tiling=NoTile
      imageSize=<32 32>
      margin=<ltrb 0 0 0 0>
      outlineSettings {
       bUseBrushTransparency=false
       roundingType=HalfHeightRadius
       width=0
       color {colorUseRule=UseColor_Specified, specifiedColor=<rgba 0 0 0 0>}
       cornerRadii=<0 0 0 0>
      }
      tintColor {colorUseRule=UseColor_Specified, specifiedColor=<rgba 1 1 1 1>}
      uVRegion {bIsValid=false, max=<0 0>, min=<0 0>}
     }
    }
   }
   slot {class=VerticalBoxSlot}
  }
  Text_RichPresence {
   widgetClassPath=&CommonTextBlock
   bIsVariable=true
   parent=VerticalBox_0
   widget=Text_RichPresence
   properties {
    clipping=ClipToBounds
    scrollStyle=&TextScrollStyle-Base_C
    text="(PH) Action Text"
    textOverflowPolicy=Ellipsis
    visibility=HitTestInvisible
    colorAndOpacity {colorUseRule=UseColor_Specified, specifiedColor=<rgba 0.0369 0.9823 0.227 1>}
    font {
     bForceMonospaced=false
     bMaterialIsStencil=false
     fontMaterial=null
     fontObject=&BurbankSmall
     letterSpacing=0
     monospacedWidth=1
     size=15.75
     skewAmount=0
     typefaceFontName=Black
     outlineSettings {
      bApplyOutlineToDropShadows=false
      bMiteredCorners=false
      bSeparateFillAlpha=false
      outlineMaterial=null
      outlineSize=0
      outlineColor=<rgba 0 0 0 1>
     }
    }
    shapedTextOptions {bOverride_TextFlowDirection=true, bOverride_TextShapingMethod=false, textFlowDirection=Culture, textShapingMethod=Auto}
   }
   slot {class=VerticalBoxSlot}
  }
  SqueegeeSlot_Tag {
   widgetClassPath=/Game/UI/Plugin/Squeegee/SqueegeeInjectionSlot
   bIsVariable=true
   namedSlotHost=WBP_Discover_Tile_Base
   widget=SqueegeeSlot_Tag
   properties {
    pluginId=Squeegee
    pluginWidgetId=RatingTag
    styleParams {bShowDetails=true, fontSize=15, paddingSize=0.5, iconSize=<26 26>, iconSkewSize=<5 26>}
   }
  }
  InputActionWidget {widgetClassPath=&CommonActionWidget, bInherited=true, widget=null}
 ]
}
