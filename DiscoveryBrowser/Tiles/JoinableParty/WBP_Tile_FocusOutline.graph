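# Focus outline widget for Joinable Party tiles in the Discovery Browser, showing a rounded outline around the tile.
# A UIKit Block Outline fills a single-cell grid, with negative padding pushing it beyond the tile bounds.
# Custom hover and unhover events request the outline's Focused and Unfocused visual states, leaving the transition itself to the UIKit block.
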WidgetBlueprint /DiscoveryBrowser/Tiles/JoinableParty/WBP_Tile_FocusOutline
graphs [EventGraph {nodes [
   OnBaseTileHover() {is=title K2Node_CustomEvent}
   n1 = Get Block_Outline() -> (Block_Outline:WBP UIKit Block Outline Object Reference)
   > RequestTransition!(execute, self:UIKit Block Object Reference=@n1, State:EUIKitBlockVisualState Enum=Focused) -> (then)
   OnBaseTileUnhovered() {is=title K2Node_CustomEvent}
   > RequestTransition(self=@n1, State=Unfocused)
   Event PreConstruct()
   Event Construct()
   Event Tick()
  ]}]
variables []
widgets {
 info {inheritedWidgetCount=0, parentClass=&UserWidget, rootWidgetClass=&GridPanel, widgetCount=2}
 widgets [
  Tile {widgetClassPath=&GridPanel, widget=Tile, properties {columnFill [1], rowFill [1]}}
  Block_Outline {
   widgetClassPath=/FortUILibrary/UIKit/Buttons/Blocks/WBP_UIKit_Block_Outline
   bIsVariable=true
   parent=Tile
   widget=Block_Outline
   properties {bottomLeftCornerRadius=0.113, bottomRightCornerRadius=0.113, topLeftCornerRadius=0.113, topRightCornerRadius=0.113}
   slot {class=GridSlot, properties {padding=<ltrb -8 -8 -8 -8>}}
  }
 ]
}
